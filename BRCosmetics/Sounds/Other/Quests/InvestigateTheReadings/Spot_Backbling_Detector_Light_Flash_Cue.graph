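# Sound cue for the light flash of a detector back bling, filed under the Investigate The Readings quest sounds.
# A single wave player feeds a continuous modulator that maps a runtime "Pitch" parameter to pitch.
# It uses its own attenuation, the reactive back bling audio property sheet and the non-critical in-game UI sound class.
# A concurrency override allows one instance at a time, so a new flash stops the oldest.
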SoundCue /BRCosmetics/Sounds/Other/Quests/InvestigateTheReadings/Spot_Backbling_Detector_Light_Flash_Cue
root=ModulatorContinuous_0
nodes [
 {id=ModulatorContinuous_0, children [WavePlayer_1], properties {pitchModulationParams {minOutput=0.95, parameterName=Pitch}}}
 {
  id=WavePlayer_1
  properties {soundWaveAssetPtr=&/Spot_Backbling_Detector_Light_Flash_02}
 }
]
properties {
 attenuationSettings=&/Spot_Backbling_Detector_Light_Flash_Att
 audioPropertiesSheet=/BRCosmetics/Sounds/Settings/PropertySheets/Reactive/APS_Cosmetic_Reactive_Backbling
 bOverrideConcurrency=true
 maxDistance=1500
 soundClassObject=&Fort_UI_InGame_NonCritical
 volumeMultiplier=1
 concurrencyOverrides {maxCount=1, resolutionRule=StopOldest}
}
